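# Translucent, unlit, two-sided smoke trail material for the DarkStance bass cosmetic's effects, usable on Niagara ribbons and mesh particles.
# Panning trail textures, warped by blurred noise and offset by particle random and vertex red, combine with caustic breakup and length fades.
# Opacity adds particle alpha, view-angle and depth fades; color is a two-color gradient with texture-driven hue shift.
# Dynamic-parameter-timed sine waves drive directional vertex wobble and breakup masks.
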Material /SparksCosmetics/Instrument/Bass/DarkStance/Effects/M_Bass_DarkStance_Smoke
graph {
 properties {
  bUsedWithMeshDeformer=true
  bUsedWithNiagaraMeshParticles=true
  bUsedWithNiagaraRibbons=true
  bUsedWithSkeletalMesh=true
  blendMode=BLEND_Translucent
  pixelDepthOffsetMode=PDOM_Legacy
  shadingModel=MSM_Unlit
  twoSided=true
  usedShadingModels=Unlit
 }
 nodes [
  Constant() {r=1}
  DynamicParameter() {defaultValue=<rgba -0.3 -0.15 1 1>, paramNames [ScrollSpeed, WPO Speed, Param3, Param4]}
  MaterialFunctionCall(Emissive=NamedRerouteUsage()) {materialFunction=&MF_FX_EmissiveAdjustments}
  n1 = TextureCoordinate() {uTiling=0.8}
  n2 = Constant2Vector() {g=0.9, r=4}
  n3 = VertexColor()
  n4 = Constant()
  n5 = Add(A=Add(A=Multiply(A=@n1, B=@n2), B=AppendVector(A=@n3.R, B=@n4)), B=AppendVector(A=ParticleRandom(), B=Constant()))
  n6 = ParticleRandom()
  n7 = Constant()
  n8 = Add(A=Add(A=TextureCoordinate() {uTiling=0.8}, B=AppendVector(A=@n6, B=@n7)), B=AppendVector(A=VertexColor().R, B=Constant()))
  n9 = Subtract(A=TextureSample(UVs=Panner(Coordinate=@n8) {speedX=0.1}) {samplerType=SAMPLERTYPE_Grayscale, texture=&LowResBlurredNoise}.R) {constB=0.5}
  n10 = Multiply(A=@n9) {constB=0.2}
  n11 = Panner(Coordinate=Add(A=Add(A=@n5, B=@n10) {constB=0}, B=Constant2Vector() {g=0.05}) {constB=2}, Speed=NamedRerouteUsage()) {speedX=-0.5}
  n12 = Power(Base=Max(A=TextureSample(UVs=@n11) {samplerType=SAMPLERTYPE_Grayscale, texture=&T_Smoke_Trail_Soft}.R) {constB=0}) {constExponent=1.2}
  n13 = Multiply(
   A=TextureCoordinate() {uTiling=0.8, vTiling=0.7}
   B=NamedRerouteDeclaration(None=Constant2Vector() {g=1, r=7}) {name=UVTiling, nodeColor=<rgba 0.875 1 0 1>}
  )
  n14 = AppendVector(A=VertexColor().R, B=Constant())
  n15 = ParticleRandom()
  n16 = Constant()
  n17 = Add(A=Add(A=Add(A=@n13, B=@n14), B=AppendVector(A=@n15, B=@n16)), B=Constant2Vector() {g=0.12})
  n18 = Multiply(A=@n9) {constB=0.15}
  n19 = Constant2Vector() {r=-0.75}
  n20 = TextureSample(UVs=Panner(Coordinate=Add(A=@n17, B=@n18) {constB=0}, Speed=NamedRerouteDeclaration(None=@n19) {name=PanSpeed, nodeColor=<rgba 0 0.25 1 1>}) {speedX=-0.5}) {samplerType=SAMPLERTYPE_LinearColor, texture=/BRCosmetics/Athena/Effects/Particles/Skydiving/NS_Contrail_CubeNinja/T_CubeNinja_Trail}
  n21 = Multiply(A=@n20.R)
  n22 = TextureSample(UVs=@n11) {texture=/BRCosmetics/Weapons/FORT_Melee/Pickaxe_Future_Samurai_Summer_Male/FX/T_Pickaxe_Future_Samurai_Summer_Male_Trail_01}
  n23 = Power(Base=Multiply(A=LinearInterpolate(A=@n12, B=@n21, Alpha=@n22.R) {constAlpha=0.4}) {constB=1.2}) {constExponent=1.2}
  n24 = ComponentMask(None=TextureCoordinate()) {r=true}
  n25 = Add(A=TextureCoordinate(), B=AppendVector(A=ParticleRandom(), B=Constant()))
  n26 = LinearInterpolate(Alpha=Power(Base=ComponentMask(None=@n25) {r=true}) {constExponent=1}) {constA=-1, constB=0.5}
  n27 = Multiply(A=TextureCoordinate() {vTiling=0.5}, B=Constant2Vector() {g=1, r=2})
  n28 = AppendVector(A=VertexColor().R, B=Constant())
  n29 = TextureSample(UVs=Panner(Coordinate=Add(A=@n27, B=@n28), Speed=NamedRerouteUsage()) {speedX=-0.5}) {samplerType=SAMPLERTYPE_LinearGrayscale, texture=/Game/Global/Textures/Default/Caustics/T_Nosie_Caustics_Grayscale}
  n30 = Saturate(None=LinearInterpolate(Alpha=Power(Base=@n29.R, Exp=@n26)) {constA=-1})
  n31 = OneMinus(None=ComponentMask(None=TextureCoordinate()) {r=true})
  n32 = Saturate(None=Multiply(A=Add(A=@n31) {constB=-0.1}) {constB=3})
  n33 = Multiply(
   A=Multiply(A=Saturate(None=@n23), B=LinearInterpolate(Alpha=@n22.R) {constA=0.5})
   B=Multiply(A=Saturate(None=@n24), B=Multiply(A=@n30, B=@n32))
  ) {constA=3}
  Multiply(A=@n33) {constB=2}
  n34 = ParticleRandom()
  n35 = Constant()
  n36 = VertexColor()
  n37 = Multiply(A=Add(A=TextureCoordinate() {uTiling=0.7}, B=AppendVector(A=@n34, B=@n35)), B=LinearInterpolate(Alpha=@n36.R) {constA=0.7})
  n38 = Add(A=Time(), B=LinearInterpolate(Alpha=@n36.R) {constA=-0.5})
  n39 = DynamicParameter() {defaultValue=<rgba -0.3 -0.15 1 1>, paramNames [ScrollSpeed, WPO Speed, Param3, Param4]}
  n40 = Sine(None=Add(A=ComponentMask(None=@n37) {r=true}, B=Multiply(A=@n38, B=@n39.ScrollSpeed) {constB=-0.3, desc=Speed of breakup masks})) {period=0.75}
  n41 = ParticleRandom()
  n42 = Constant()
  n43 = VertexColor()
  n44 = Multiply(A=Add(A=TextureCoordinate() {uTiling=0.7}, B=AppendVector(A=@n41, B=@n42)), B=LinearInterpolate(Alpha=@n43.R) {constA=0.7})
  n45 = Add(A=Time(), B=LinearInterpolate(Alpha=@n43.R) {constA=-0.5})
  n46 = Sine(None=Add(A=ComponentMask(None=@n44) {r=true}, B=Multiply(A=@n45, B=@n39.WPO Speed) {constB=-0.3, desc=Speed of breakup masks})) {period=0.5}
  Multiply(A=@n46, B=@n40)
  NamedRerouteDeclaration(None=Saturate(None=Multiply(A=@n40))) {name=AlphaSineMask, nodeColor=<rgba 1 0.75 0 1>}
  n47 = Constant() {r=5}
  n48 = Multiply(
   A=Transform(None=VectorParameter() {parameterName=WPO Direction1, defaultValue=<rgba 0 1 0 1>}.RGB) {transformSourceType=TRANSFORMSOURCE_Local}
   B=LinearInterpolate(Alpha=@n46)
  )
  n49 = Multiply(
   A=LinearInterpolate(Alpha=@n46) {constA=0.3, constB=1.3}
   B=Transform(None=VectorParameter() {parameterName=WPO Direction2, defaultValue=<rgba 1 0 0 1>}.RGB) {transformSourceType=TRANSFORMSOURCE_Local}
  )
  n50 = ComponentMask(None=TextureCoordinate()) {r=true}
  n51 = Multiply(
   A=Multiply(A=Add(A=@n48, B=@n49), B=ScalarParameter() {defaultValue=2, parameterName=WPO Intensity})
   B=Saturate(None=Multiply(A=@n50) {constB=3})
  )
  NamedRerouteDeclaration(None=Add(
    A=@n51
    B=MaterialFunctionCall(Offset Amount=@n47) {materialFunction=/Engine/Functions/Engine_MaterialFunctions02/WorldPositionOffset/CameraOffset}.Camera Offset
   ) {constB=0}) {name=WPO Out, nodeColor=<rgba 1 0.6797 0 1>}
  n52 = LinearInterpolate(Alpha=TextureSample() {texture=&T_Inky_Smoke_Tile}.R) {constA=-0.02, constB=0.05}
  n53 = ScalarParameter() {defaultValue=-0.01, parameterName=Color offset}
  n54 = ComponentMask(None=TextureCoordinate()) {r=true}
  n55 = LinearInterpolate(
   A=VectorParameter() {parameterName=Color1, defaultValue=<rgba 0 0.7 0.1968 1>}.RGB
   B=VectorParameter() {parameterName=Color2, defaultValue=<rgba 0 1 0.2812 1>}.RGB
   Alpha=Power(Base=OneMinus(None=@n54), Exp=ScalarParameter() {defaultValue=1, parameterName=ColorRangePower}) {constExponent=1}
  )
  NamedRerouteDeclaration(None=MaterialFunctionCall(Hue Shift Percentage=Add(A=@n52, B=@n53) {constB=0}, Texture=Multiply(A=@n55) {constB=0.5}) {materialFunction=&HueShift}.Result) {name=Emissive Color, nodeColor=<rgba 0.3359 1 0 1>}
  n56 = Multiply(A=Saturate(None=@n33), B=NamedRerouteUsage()) {desc=Mid segment breakup}
  n57 = Multiply(
   A=Multiply(A=ParticleColor().A, B=Multiply(A=@n56) {constB=2.5})
   B=MaterialFunctionCall("FadeSensitivity (optional)"=ScalarParameter() {defaultValue=2, parameterName=View Angle Fade Sensitivity}) {materialFunction=&FadeBasedOnViewAngle}.Result
  )
  NamedRerouteDeclaration(None=Saturate(None=DepthFade(Opacity=@n57) {fadeDistanceDefault=5})) {name=Opacity Out, nodeColor=<rgba 0 0.625 1 1>}
  NamedRerouteDeclaration(None=@n29.R) {name=Breakup Pattern, nodeColor=<rgba 0.0078 0 1 1>}
  NamedRerouteUsage()
  NamedRerouteUsage()
  Reroute(None=NamedRerouteUsage())
  Reroute(None=Add(A=WorldPosition().XYZ, B=@n51))
  Reroute(None=NamedRerouteUsage())
  n58 = Constant() {r=10}
  n59 = LinearInterpolate(Alpha=VertexColor().R) {constA=0.7, constB=1.2}
  Reroute(None=MaterialFunctionCall(WidthBase=Multiply(A=@n58, B=@n59), WidthTip=Constant() {r=25}, Expand U or V UV Channel=StaticBool()) {materialFunction=&SplineThicken}.WorldPositionOffset)
  StaticBool()
 ]
}
