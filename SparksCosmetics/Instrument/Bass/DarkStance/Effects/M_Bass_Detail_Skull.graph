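# Unlit translucent material for skull detail effects on the DarkStance bass cosmetic, set up for Niagara mesh particles.
# A skull texture passes through adjustable desaturation, is tinted by particle color, then is scaled and fed through a shared FX emissive adjustment function.
# Particle alpha multiplies a view-angle fade with adjustable sensitivity, forming a fade mask that varies with viewing angle.
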Material /SparksCosmetics/Instrument/Bass/DarkStance/Effects/M_Bass_Detail_Skull
graph {
 properties {
  bUsedWithNiagaraMeshParticles=true
  blendMode=BLEND_Translucent
  pixelDepthOffsetMode=PDOM_Legacy
  previewMesh=&/SM_Bass_Detail_Skull
  shadingModel=MSM_Unlit
  usedShadingModels=Unlit
 }
 nodes [
  Constant() {r=1}
  n1 = Multiply(
   A=ParticleColor().A
   B=MaterialFunctionCall("FadeSensitivity (optional)"=ScalarParameter() {defaultValue=1.75, parameterName="FadeSensitivity (optional) (S)"}) {materialFunction=&FadeBasedOnViewAngle}.Result
  )
  LinearInterpolate(A=@n1, B=Constant() {r=1}) {constAlpha=0.75}
  n2 = Desaturation(
   None=TextureSample() {texture=/Game/Effects/Fort_Effects/GameplayPulginsAssets/Skull/SkullDetail/Texture/T_DetailSkull_E}.RGB
   Fraction=ScalarParameter() {parameterName=Des}
  )
  n3 = ParticleColor()
  MaterialFunctionCall(Emissive=Multiply(A=Multiply(A=@n3.RGB, B=@n2), B=ScalarParameter() {defaultValue=1.25, parameterName=Emissive}) {constB=2}) {materialFunction=&MF_FX_EmissiveAdjustments}
  Reroute(None=@n1)
  Reroute(None=Multiply(A=@n2, B=@n3.RGB))
 ]
}
