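# Two-sided translucent unlit Niagara sprite and mesh particle material drawing a soft radial gradient for the DarkStance bass cosmetic effects.
# Emissive comes from particle color scaled by time of day.
# Opacity multiplies a sphere mask, shaped by a dynamic Power parameter, by particle alpha and switchable view-angle fade and panning two-layer noise smoke mask.
# Depth and near-camera fades apply only at High and Epic quality.
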Material /SparksCosmetics/Instrument/Bass/DarkStance/Effects/M_Radial_Gradient_Trans_2Sided
graph {
 properties {
  bUsedWithNanite=true
  bUsedWithNiagaraMeshParticles=true
  bUsedWithNiagaraSprites=true
  blendMode=BLEND_Translucent
  pixelDepthOffsetMode=PDOM_Legacy
  shadingModel=MSM_Unlit
  twoSided=true
  usedShadingModels=Unlit
 }
 nodes [
  n1 = ParticleColor()
  n2 = MaterialFunctionCall(
   Value_Morning=ScalarParameter() {defaultValue=1, parameterName="Value_Morning (S)"}
   Value_Day=ScalarParameter() {defaultValue=1, parameterName="Value_Day (S)"}
   Value_Evening=ScalarParameter() {defaultValue=1, parameterName="Value_Evening (S)"}
   Value_Night=ScalarParameter() {defaultValue=1, parameterName="Value_Night (S)"}
  ) {materialFunction=&Scalar_Time_of_Day}
  MaterialFunctionCall(Emissive=Multiply(A=Multiply(A=@n1.RGB, B=@n2.Result)) {constB=2.5}) {materialFunction=&MF_FX_EmissiveAdjustments}
  n3 = TextureCoordinate()
  n4 = Constant() {r=0.5}
  n5 = ScalarParameter() {defaultValue=0.5, parameterName=Radius}
  n6 = ScalarParameter() {parameterName=Hardness}
  n7 = DynamicParameter() {defaultValue=<rgba 3 1 1 1>, paramNames [Power, DepthFadeDistMult, Param3, Param4]}
  n8 = Constant() {r=1}
  n9 = DepthFade(
   Opacity=MaterialFunctionCall(Distance=ScalarParameter() {defaultValue=100, parameterName=NearCameraFadeDist}) {materialFunction=&Near_Camera_Fade}.Result
   FadeDistance=Multiply(A=@n7.DepthFadeDistMult, B=ScalarParameter() {defaultValue=100, parameterName=DepthFadeDist})
  )
  n10 = Multiply(
   A=Power(
    Base=SphereMask(A=@n3, B=@n4, Radius=@n5, Hardness=@n6) {hardnessPercent=0}
    Exp=Clamp(None=@n7.Power) {clampMode=CMODE_ClampMin, maxDefault=100, minDefault=1}
   ) {constExponent=3}
   B=Multiply(A=@n1.A, B=QualitySwitch(Default=@n8, Low=@n8, High=@n9, Medium=@n8, Epic=@n9))
  )
  n11 = StaticSwitchParameter(
   True=MaterialFunctionCall("FadeSensitivity (optional)"=ScalarParameter() {defaultValue=7, parameterName=AngleFade}) {materialFunction=&FadeBasedOnViewAngle}.Result
   False=Constant() {r=1}
  ) {defaultValue=true, parameterName=Use AngleFade}
  n12 = Constant() {r=1}
  n13 = TextureCoordinate()
  n14 = Multiply(A=@n13, B=Constant2Vector() {g=2.5, r=1})
  n15 = Multiply(A=@n13, B=Constant2Vector() {g=1.5, r=0.75})
  n16 = Multiply(
   A=TextureSample(UVs=Panner(Coordinate=@n14) {speedX=0.1, speedY=0.5}) {samplerType=SAMPLERTYPE_Grayscale, texture=&LowResBlurredNoise}.R
   B=TextureSample(UVs=Panner(Coordinate=@n15) {speedX=-0.05, speedY=0.325}) {samplerType=SAMPLERTYPE_Grayscale, texture=&LowResBlurredNoise}.R
  )
  n17 = Add(A=TextureCoordinate(), B=Constant2Vector() {g=0.15})
  n18 = Multiply(A=OneMinus(None=ComponentMask(None=@n17) {g=true})) {constB=10}
  n19 = Multiply(A=Multiply(A=Power(Base=@n16) {constExponent=0.5}, B=Saturate(None=@n18))) {constB=4}
  Saturate(None=Multiply(A=Multiply(A=@n10, B=@n11), B=StaticSwitchParameter(True=@n19, False=@n12) {defaultValue=true, parameterName=Use Smoke Mask}) {desc=Smoke Mask})
 ]
}
